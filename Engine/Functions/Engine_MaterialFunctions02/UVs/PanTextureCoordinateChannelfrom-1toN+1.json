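# Material function that pans one 0-1 UV channel from -1 to the tile count +1, intended for textures with clamped coordinates.
# The UV channel input, defaulting to texture coordinate U, is multiplied by Tiling Amount and offset by an Add's default constant.
# Time (0-1), defaulting to fractional time from TimeWithSpeedVariable, scales the similarly offset tiling amount, which is subtracted to slide the coordinate across the full range.
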{"asset":"/Engine/Functions/Engine_MaterialFunctions02/UVs/PanTextureCoordinateChannelfrom-1toN+1.PanTextureCoordinateChannelfrom-1ton+1","class":"MaterialFunction","graph":{"expressions":{"MaterialExpressionAdd_26":{"class":"MaterialExpressionAdd","inputs":[{"from":"MaterialExpressionMultiply_28","name":"A","output":""},{"from":"","name":"B","output":""}],"properties":{}},"MaterialExpressionAdd_27":{"class":"MaterialExpressionAdd","inputs":[{"from":"MaterialExpressionFunctionInput_2","name":"A","output":""},{"from":"","name":"B","output":""}],"properties":{}},"MaterialExpressionComponentMask_13":{"class":"MaterialExpressionComponentMask","inputs":[{"from":"MaterialExpressionTextureCoordinate_9","name":"None","output":""}],"properties":{"r":true}},"MaterialExpressionFunctionInput_0":{"class":"MaterialExpressionFunctionInput","inputs":[{"from":"MaterialExpressionComponentMask_13","name":"Preview","output":""}],"properties":{"bUsePreviewValueAsDefault":true,"inputName":"UV Channel","inputType":"FunctionInput_Scalar"}},"MaterialExpressionFunctionInput_2":{"class":"MaterialExpressionFunctionInput","inputs":[{"from":"","name":"Preview","output":""}],"properties":{"bUsePreviewValueAsDefault":true,"inputName":"Tiling Amount","inputType":"FunctionInput_Scalar","previewValue":{"w":1,"x":5,"y":0,"z":0},"sortPriority":1}},"MaterialExpressionFunctionInput_4":{"class":"MaterialExpressionFunctionInput","inputs":[{"from":"MaterialExpressionMaterialFunctionCall_10","name":"Preview","output":"Frac Time"}],"properties":{"bUsePreviewValueAsDefault":true,"inputName":"Time (0-1)","inputType":"FunctionInput_Scalar","previewValue":{"w":1,"x":5,"y":0,"z":0},"sortPriority":2}},"MaterialExpressionFunctionOutput_0":{"class":"MaterialExpressionFunctionOutput","inputs":[{"from":"MaterialExpressionSubtract_20","name":"None","output":""}],"properties":{"outputName":"UV Channel output"}},"MaterialExpressionMaterialFunctionCall_10":{"class":"MaterialExpressionMaterialFunctionCall","inputs":[{"from":"","name":"Time","output":""},{"from":"","name":"Speed","output":""}],"properties":{"desc":"0-1","materialFunction":"/Engine/Functions/Engine_MaterialFunctions02/Math/TimeWithSpeedVariable.TimeWithSpeedVariable"}},"MaterialExpressionMultiply_28":{"class":"MaterialExpressionMultiply","inputs":[{"from":"MaterialExpressionFunctionInput_0","name":"A","output":""},{"from":"MaterialExpressionFunctionInput_2","name":"B","output":""}],"properties":{}},"MaterialExpressionMultiply_29":{"class":"MaterialExpressionMultiply","inputs":[{"from":"MaterialExpressionAdd_27","name":"A","output":""},{"from":"MaterialExpressionFunctionInput_4","name":"B","output":""}],"properties":{}},"MaterialExpressionSubtract_20":{"class":"MaterialExpressionSubtract","inputs":[{"from":"MaterialExpressionAdd_26","name":"A","output":""},{"from":"MaterialExpressionMultiply_29","name":"B","output":""}],"properties":{}},"MaterialExpressionTextureCoordinate_9":{"class":"MaterialExpressionTextureCoordinate","inputs":[],"properties":{}}},"properties":{"bExposeToLibrary":true,"description":"This function pans a 0-1 texture coordinate channel from -1 to the number of tiled units +1. This is useful when used in conjunction with textures with clamped coordinates.","libraryCategoriesText":["Misc","Coordinates"],"thumbnailInfo":"<self>:SceneThumbnailInfoWithPrimitive_30"}},"schema":"material/v2"}
