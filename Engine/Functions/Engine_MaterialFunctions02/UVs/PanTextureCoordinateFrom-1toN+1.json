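# Material function, exposed to the library, that pans 0-1 UVs from -1 to the tiling count plus one, for use with clamped-coordinate textures.
# It splits the UVs, UV Tiling and Time (0-1) inputs into components and runs PanTextureCoordinateChannelfrom-1toN+1 separately on R and G.
# The two channel results are appended into the UVs output.
# UVs default to texture coordinates and time to TimeWithSpeedVariable's Frac Time.
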{"asset":"/Engine/Functions/Engine_MaterialFunctions02/UVs/PanTextureCoordinateFrom-1toN+1.PanTextureCoordinateFrom-1toN+1","class":"MaterialFunction","graph":{"expressions":{"MaterialExpressionAppendVector_8":{"class":"MaterialExpressionAppendVector","inputs":[{"from":"MaterialExpressionMaterialFunctionCall_12","name":"A","output":"UV Channel output"},{"from":"MaterialExpressionMaterialFunctionCall_14","name":"B","output":"UV Channel output"}],"properties":{}},"MaterialExpressionFunctionInput_11":{"class":"MaterialExpressionFunctionInput","inputs":[{"from":"MaterialExpressionTextureCoordinate_11","name":"Preview","output":""}],"properties":{"bUsePreviewValueAsDefault":true,"inputName":"UVs","inputType":"FunctionInput_Vector2","previewValue":{"w":1,"x":5,"y":0,"z":0}}},"MaterialExpressionFunctionInput_13":{"class":"MaterialExpressionFunctionInput","inputs":[{"from":"MaterialExpressionMaterialFunctionCall_21","name":"Preview","output":"Frac Time"}],"properties":{"bUsePreviewValueAsDefault":true,"inputName":"Time (0-1)","inputType":"FunctionInput_Vector2","previewValue":{"w":1,"x":5,"y":0,"z":0},"sortPriority":2}},"MaterialExpressionFunctionInput_6":{"class":"MaterialExpressionFunctionInput","inputs":[{"from":"","name":"Preview","output":""}],"properties":{"bUsePreviewValueAsDefault":true,"inputName":"UV Tiling","inputType":"FunctionInput_Vector2","previewValue":{"w":1,"x":1,"y":1,"z":0},"sortPriority":1}},"MaterialExpressionFunctionOutput_2":{"class":"MaterialExpressionFunctionOutput","inputs":[{"from":"MaterialExpressionAppendVector_8","name":"None","output":""}],"properties":{"outputName":"UVs"}},"MaterialExpressionMaterialFunctionCall_12":{"class":"MaterialExpressionMaterialFunctionCall","inputs":[{"from":"MaterialExpressionMaterialFunctionCall_15","name":"UV Channel","output":"R"},{"from":"MaterialExpressionMaterialFunctionCall_17","name":"Tiling Amount","output":"R"},{"from":"MaterialExpressionMaterialFunctionCall_19","name":"Time (0-1)","output":"R"}],"properties":{"materialFunction":"/Engine/Functions/Engine_MaterialFunctions02/UVs/PanTextureCoordinateChannelfrom-1toN+1.PanTextureCoordinateChannelfrom-1ton+1"}},"MaterialExpressionMaterialFunctionCall_14":{"class":"MaterialExpressionMaterialFunctionCall","inputs":[{"from":"MaterialExpressionMaterialFunctionCall_15","name":"UV Channel","output":"G"},{"from":"MaterialExpressionMaterialFunctionCall_17","name":"Tiling Amount","output":"G"},{"from":"MaterialExpressionMaterialFunctionCall_19","name":"Time (0-1)","output":"G"}],"properties":{"materialFunction":"/Engine/Functions/Engine_MaterialFunctions02/UVs/PanTextureCoordinateChannelfrom-1toN+1.PanTextureCoordinateChannelfrom-1ton+1"}},"MaterialExpressionMaterialFunctionCall_15":{"class":"MaterialExpressionMaterialFunctionCall","inputs":[{"from":"MaterialExpressionFunctionInput_11","name":"Float2","output":""}],"properties":{"materialFunction":"/Engine/Functions/Engine_MaterialFunctions02/Utility/BreakOutFloat2Components.BreakOutFloat2Components"}},"MaterialExpressionMaterialFunctionCall_17":{"class":"MaterialExpressionMaterialFunctionCall","inputs":[{"from":"MaterialExpressionFunctionInput_6","name":"Float2","output":""}],"properties":{"materialFunction":"/Engine/Functions/Engine_MaterialFunctions02/Utility/BreakOutFloat2Components.BreakOutFloat2Components"}},"MaterialExpressionMaterialFunctionCall_19":{"class":"MaterialExpressionMaterialFunctionCall","inputs":[{"from":"MaterialExpressionFunctionInput_13","name":"Float2","output":""}],"properties":{"materialFunction":"/Engine/Functions/Engine_MaterialFunctions02/Utility/BreakOutFloat2Components.BreakOutFloat2Components"}},"MaterialExpressionMaterialFunctionCall_21":{"class":"MaterialExpressionMaterialFunctionCall","inputs":[{"from":"","name":"Time","output":""},{"from":"","name":"Speed","output":""}],"properties":{"desc":"0-1","materialFunction":"/Engine/Functions/Engine_MaterialFunctions02/Math/TimeWithSpeedVariable.TimeWithSpeedVariable"}},"MaterialExpressionTextureCoordinate_11":{"class":"MaterialExpressionTextureCoordinate","inputs":[],"properties":{}}},"properties":{"bExposeToLibrary":true,"description":"This function pans 0-1 uv coordinates from -1 to the number of tiled units +1. This is useful when used in conjunction with textures with clamped coordinates.","libraryCategoriesText":["Misc","Coordinates"],"thumbnailInfo":"<self>:SceneThumbnailInfoWithPrimitive_30"}},"schema":"material/v2"}
