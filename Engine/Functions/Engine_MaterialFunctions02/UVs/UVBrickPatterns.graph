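# UVBrickPatterns is a library material function that turns 2D coordinates into an offset, brick-like grid.
# A static switch picks X or Y as the shifted axis.
# The other axis plus Offset Percentage is rounded up by Ceil, scaled by that percentage and added to the shifted axis, staggering each row or column cumulatively.
# Both channels then subtract one minus the offset, output as Non-frac and as Frac-wrapped per-cell coordinates.
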MaterialFunction /Engine/Functions/Engine_MaterialFunctions02/UVs/UVBrickPatterns
graph {
 properties {
  bExposeToLibrary=true
  description="Takes coordinates and manipulates thems to form an offset grid patten much like bricks. "
  libraryCategoriesText [Misc,Coordinates]
 }
 nodes [
  n1 = FunctionInput(Preview=TextureCoordinate()) {
   bUsePreviewValueAsDefault=true
   description="Enter world position channels, uvs or any other data. "
   inputName=Float 2 Coordinates
   inputType=FunctionInput_Vector2
   previewValue=<15 15 0 1>
  }
  n2 = ComponentMask(None=@n1) {g=true}
  n3 = ComponentMask(None=@n1) {r=true}
  n4 = FunctionInput(Preview=StaticBool()) {
   bUsePreviewValueAsDefault=true
   description="Enter world position channels, uvs or any other data. "
   inputName="Offset X (True) Or Offset Y (False)"
   inputType=FunctionInput_StaticBool
   previewValue=<15 15 0 1>
  }
  n5 = StaticSwitch(True=@n2, False=@n3, Value=@n4)
  n6 = FunctionInput() {
   bUsePreviewValueAsDefault=true
   description=Amount each cell will be offset
   inputName=Offset Percentage
   inputType=FunctionInput_Scalar
   previewValue=<0.5 0 0 1>
  }
  n7 = Add(A=@n5, B=@n6) {constB=0.5}
  n8 = Add(A=Multiply(A=Ceil(None=@n7), B=@n6) {constB=0.5}, B=StaticSwitch(True=@n3, False=@n2, Value=@n4))
  n9 = Subtract(A=StaticSwitch(True=AppendVector(A=@n8, B=@n5), False=AppendVector(A=@n5, B=@n8), Value=@n4), B=OneMinus(None=@n6)) {constB=0.5}
  FunctionOutput(None=@n9) {outputName=Non-frac, sortPriority=1}
  FunctionOutput(None=Frac(None=@n9) {desc=mod}) {outputName=Frac}
 ]
}
