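# Material function, exposed to the library, that scales world position for texturing use.
# It divides WorldPosition by a Divisor input, which defaults to one when unconnected.
# Component masks split the result into XYZ, XY and Z outputs for common configurations.
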MaterialFunction /Engine/Functions/Engine_MaterialFunctions02/Texturing/WorldPositionWithScale
graph {
 properties {
  bExposeToLibrary=true
  description=WorldPosition divided by a number and split into common configurations.
  libraryCategoriesText [Misc,Texturing]
 }
 nodes [
  n1 = Divide(A=WorldPosition() {desc=change}.XYZ, B=FunctionInput() {bUsePreviewValueAsDefault=true, inputName=Divisor, previewValue=<1 1 1 1>})
  FunctionOutput(None=ComponentMask(None=@n1) {b=true, g=true, r=true}) {outputName=XYZ}
  FunctionOutput(None=ComponentMask(None=@n1) {g=true, r=true}) {outputName=XY, sortPriority=1}
  FunctionOutput(None=ComponentMask(None=@n1) {b=true}) {outputName=Z, sortPriority=2}
 ]
}
