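# Material function that produces a locally aligned 0-1 UVW coordinate spanning an object's bounding box, exposed to the library under Misc and Coordinates.
# It transforms world position, excluding shader offsets, into instance space, subtracts the local bounds minimum from ObjectLocalBounds and divides by the bounds size.
# The result is output as a combined RGB vector and as separate R, G and B channels.
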MaterialFunction /Engine/Functions/Engine_MaterialFunctions02/UVs/BoundingBoxBased_0-1_UVW
graph {
 properties {bExposeToLibrary=true, description=Creates a locally aligned 0-1 value on XY and Z., libraryCategoriesText [Misc,Coordinates]}
 nodes [
  n1 = WorldPosition() {worldPositionShaderOffset=WPT_ExcludeAllShaderOffsets}
  n2 = MaterialFunctionCall() {materialFunction=&ObjectLocalBounds}
  n3 = Divide(
   A=Subtract(
    A=TransformPosition(None=@n1.XYZ) {transformSourceType=TRANSFORMPOSSOURCE_World, transformType=TRANSFORMPOSSOURCE_Instance}
    B=@n2.Local Bounds Minimum
   )
   B=@n2.Local Bounds Size
  )
  FunctionOutput(None=@n3) {outputName=RGB}
  n4 = MaterialFunctionCall(Float3=@n3) {materialFunction=&BreakOutFloat3Components}
  FunctionOutput(None=@n4.R) {outputName=R, sortPriority=1}
  FunctionOutput(None=@n4.G) {outputName=G, sortPriority=2}
  FunctionOutput(None=@n4.B) {outputName=B, sortPriority=3}
 ]
}
