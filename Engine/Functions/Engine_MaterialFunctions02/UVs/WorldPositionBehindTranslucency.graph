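# Material function that reconstructs the scene's world position behind a translucent surface, exposed to the material function library.
# It scales the camera-to-pixel vector by the ratio of scene depth to pixel depth, then adds the camera position back.
# Optional inputs override the scene depth sample and exclude World Position Offset from the surface position, which is included by default.
# A camera-direction dot product branch is left unconnected to the output.
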MaterialFunction /Engine/Functions/Engine_MaterialFunctions02/UVs/WorldPositionBehindTranslucency
graph {
 properties {
  bExposeToLibrary=true
  description="This will recreate the scenes world position behind a translucent surface. "
  libraryCategoriesText [Misc,Texturing]
 }
 nodes [
  DotProduct(
   A=MaterialFunctionCall() {materialFunction=&CameraDirectionVector}.Result
   B=Subtract(
    A=MaterialFunctionCall!(Divisor) {materialFunction=/Engine/Functions/Engine_MaterialFunctions02/Texturing/WorldPositionWithScale}.XYZ
    B=CameraPositionWS()
   )
  )
  n1 = FunctionInput(Preview=SceneDepth().R) {
   bUsePreviewValueAsDefault=true
   description=Allows overriding scene depth with custom sampling.
   inputName=SceneDepth
   inputType=FunctionInput_Scalar
  }
  n2 = Divide(A=@n1, B=PixelDepth().R)
  n3 = FunctionInput(Preview=StaticBool()) {
   bUsePreviewValueAsDefault=true
   description=False by default.
   inputName=Exclude WPO Offsets
   inputType=FunctionInput_StaticBool
   sortPriority=1
  }
  n4 = CameraPositionWS()
  n5 = Subtract(
   A=StaticSwitch(True=WorldPosition() {worldPositionShaderOffset=WPT_ExcludeAllShaderOffsets}.XYZ, False=WorldPosition().XYZ, Value=@n3)
   B=@n4
  )
  FunctionOutput(None=Add(A=Multiply(A=@n2, B=@n5), B=@n4))
  PixelDepth()
 ]
}
