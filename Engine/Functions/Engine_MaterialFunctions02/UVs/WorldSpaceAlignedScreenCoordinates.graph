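# Material function that re-centres screen-aligned UVs on a world position, which defaults to the component's pivot.
# It subtracts the point's clip-space-derived buffer UV from viewport UV, then multiplies by camera distance.
# An X correction divides buffer resolution R by itself, so it has no effect, giving Offset UVs.
# A second output multiplies that by FOV over view width for consistency across resolutions.
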MaterialFunction /Engine/Functions/Engine_MaterialFunctions02/UVs/WorldSpaceAlignedScreenCoordinates
graph {
 properties {
  bExposeToLibrary=true
  description=Moves the center of screen aligned coordinates to the center of a world position. A second output scales the results based on the users fov and view size to maintain a consistent look across multiple resolutions.
  libraryCategoriesText [Coordinates]
 }
 nodes [
  n1 = CameraPositionWS()
  n2 = FunctionInput(Preview=MaterialFunctionCall() {materialFunction=&ComponentPivotLocation}.Result) {bUsePreviewValueAsDefault=true, desc=test, description=Defaults to the component's pivot point., inputName=Center Point World Position}
  n3 = ScreenPosition()
  n4 = MaterialFunctionCall(World Position=@n2) {materialFunction=&TransformToClipSpace}
  n5 = MaterialFunctionCall(Float2=MaterialFunctionCall() {materialFunction=&ScreenResolution}.Buffer Resolution) {materialFunction=&BreakOutFloat2Components}
  n6 = Multiply(
   A=Multiply(A=Distance(A=@n1, B=@n2), B=Subtract(A=@n3.ViewportUV, B=@n4.Buffer UVs for Scene Textures))
   B=AppendVector(A=Divide(A=@n5.R, B=@n5.R), B=Constant() {r=1})
  )
  n7 = MaterialFunctionCall() {materialFunction=/Engine/Functions/Engine_MaterialFunctions02/Utility/FOV}
  n8 = ComponentMask(None=ViewSize()) {r=true}
  FunctionOutput(None=Multiply(A=FunctionOutput(None=@n6) {outputName=Offset UVs}, B=Divide(A=@n7."FOV ", B=@n8))) {
   description="This output is scaled based on the camera's fov and the view size. Which will produce more predictable results. For most use cases, this value should be multiplied by a large number."
   outputName=Offset and Scaled UVS
   sortPriority=1
  }
 ]
}
